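# Niagara system for the Ice Shards cosmetic vehicle boost, emitting icy sprite effects from the car's skeletal mesh sockets, partitioned per boost port.
# BaseMist adds a short mist trail, Shards throws elongated ice streaks and falling pieces, and ImpactBits spawns drifting debris and dust behind and below the car after a short delay.
# Each emitter uses spawn groups with switched parameters and visibility tags to feed multiple sub-UV sprite renderers.
# Painted-variant overrides recolor every layer, and a frontend-preview flag adjusts velocities.
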NiagaraSystem /VehicleCosmetics/Boosts/Boost_IceShards/FX/NS_Boost_IceShards
dependencies {
 usedDataInterfaces [
  &NiagaraDataInterfaceCurve
  &NiagaraDataInterfaceSpriteRendererInfo
  &NiagaraDataInterfaceArrayFloat2
  &NiagaraDataInterfaceSkeletalMesh
 ]
 usedDynamicInputs [&RandomRangeVector]
 usedModules [
  &self:SystemSpawnScript.SystemScriptSource.SystemScriptGraph.NiagaraNodeAssignment_0.SetVariables_#1
  &SystemState
 ]
}
emitters {
 BaseMist {
  emitterUpdateScript [
   &EmitterState()
   &VC_FilteredSocketCount()
   &SpawnRate(SpawnRate=78, Use Spawn Probability=false)
   &SpawnRate(SpawnGroup=1, SpawnRate=45, Use Spawn Probability=false) {module=SpawnRate001}
  ]
  particleSpawnScript [
   /SetVariables(Particles.VisibilityTag=$Emitter.SpawnGroup) {isSetParameters=true}
   &VC_PaintedVariantsOverride(
    Black=<rgba 0.0191 0.0191 0.0191 1.1>
    BurntSienna=<rgba 0.5 0.3583 0.0749 0.8>
    Cobalt=<rgba 0.18 0.285 0.6 1>
    Crimson=<rgba 0.45 0.18 0.045 0.9>
    DefaultColor=<rgba 0.56 0.64 0.8 1.5>
    ForestGreen=<rgba 0 0.6 0.1 0.75>
    Grey=<rgba 0.27 0.2757 0.3 0.8>
    Lime=<rgba 0.42 0.5 0.1 1>
    Orange=<rgba 0.6 0.446 0.18 1>
    Pink=<rgba 0.55 0.385 0.4923 0.7>
    Purple=<rgba 0.4447 0.174 0.58 1>
    Saffron=<rgba 0.5 0.5 0.05 0.8>
    SkyBlue=<rgba 0.25 0.5 0.4796 0.55>
    TitaniumWhite=<rgba 0.44 0.495 0.55 0.6>
   )
   &InitializeParticle(
    Color=$Particles.Color
    Lifetime=&RandomRangeFloat(
     Minimum=&MergeHelper_FloatSwitch(0=0.2, 1=0.2, 2=?, 3=?, Index=$Emitter.SpawnGroup, Num=::NewEnumerator0)
     Maximum=&MergeHelper_FloatSwitch(0=0.45, 1=0.4, 2=?, 3=?, Index=$Emitter.SpawnGroup, Num=::NewEnumerator0)
    )
    Sprite Rotation Mode=::NewEnumerator2
    Sprite Size Mode=::NewEnumerator3
    Uniform Sprite Size=&RandomRangeFloat(
     Minimum=&MergeHelper_FloatSwitch(0=60, 1=60, 2=?, 3=?, Index=$Emitter.SpawnGroup, Num=::NewEnumerator0)
     Maximum=&MergeHelper_FloatSwitch(0=80, 1=80, 2=?, 3=?, Index=$Emitter.SpawnGroup, Num=::NewEnumerator0)
    )
    UsePositionOffset=false
    Write Material Random=true
   )
   &PartitionParticles(Number Of Partitions=$Emitter.PortCount)
   &SkeletalMeshLocation(
    Bone / Socket Index=$Particles.PartitionParticles.Partition
    Bone / Socket Sampling Mode=::NewEnumerator5
    MaskSpawnGroup=false
    Mesh Sampling Type=::NewEnumerator6
    Sampled Position Offset=&AddVectorToPosition(Position=<-5 0 0>, Vector=Linked() {linkedVariable {name=User.DTCarOffset, type=NiagaraPosition}})
    Skeletal Mesh=$System.SkeletalMesh
   )
   &ShapeLocation(MaskSpawnGroup=false, Override Seed=false, Sphere Radius=20, UseOwnerScale=true)
   &AddVelocity(
    Velocity=&MergeHelper_VectorSwitch(0=<-1000 0 0>, 1=<-1000 0 0>, 2=?, 3=?, Index=$Emitter.SpawnGroup, Num=::NewEnumerator0)
    Velocity Speed Scale=&MakeCustomFloatFromBool(Bool=$User.bIsFrontendPreview, True Float=1, False Float=0.35)
   )
  ]
  particleUpdateScript [
   &ParticleState()
   &ScaleSpriteSize(
    Uniform Curve Scale=&MergeHelper_FloatSwitch(0=1.2, 1=1, 2=?, 3=?, Index=&SpawnGroup(), Num=::NewEnumerator0)
    Uniform Curve Sprite Scale=DataInterface() {propertyValues {Curve {keys [{arriveTangent=9.4311, interpMode=RCIM_Cubic, leaveTangent=9.4311, tangentMode=RCTM_User, time=0.05, value=0.6}, <at 1 4 RCIM_Cubic>]}}}
   )
   &ScaleColor(
    Scale Alpha=&FloatFromCurve(
     FloatCurve=DataInterface() {propertyValues {Curve {keys [<at 0.5 1 RCIM_Cubic>,<at 1 0 RCIM_Cubic>]}}}
     Scale Curve=&MergeHelper_FloatSwitch(0=0.35, 1=0.3, 2=?, 3=?, Index=&SpawnGroup(), Num=::NewEnumerator0)
    )
    Scale RGB=&Multiply_VectorByFloat(Vector=<1 1 1>, Float=2)
    ScaleA=true
    ScaleRGB=true
   )
   &AccelerationForce(Acceleration=<2500 0 0>)
   &SolveForcesAndVelocity(Clamp Velocity=false, Limit Acceleration=false)
   &SubUVAnimation(
    "Sprite Renderer "=DataInterface() {propertyValues {SpriteRenderer=&self:BaseMist_0.NiagaraSpriteRendererProperties_0}}
    SubUV Animation Mode:ENiagara_SubUVLookupModeV2=::NewEnumerator2
    UseEndFrame=false
    UseStartFrame=false
   )
  ]
  rendererClasses [&NiagaraSpriteRendererProperties]
  renderers [
   {rendererClass=&NiagaraSpriteRendererProperties}
   {rendererClass=&NiagaraSpriteRendererProperties, rendererIndex=1}
  ]
 }
 ImpactBits {
  emitterUpdateScript [
   &EmitterState(Delay First Loop Only=true, Life Cycle Mode=::NewEnumerator1, Loop Delay=0.6, UseLoopDelay=true)
   &SpawnBurst_Instantaneous(Spawn Count=5, Use Spawn Probability=false, UseLoopCountLimit=false)
   &VC_FilteredSocketCount()
   &SpawnRate(Spawn Probability=0.7, SpawnRate=50, Use Spawn Probability=true)
   &SpawnRate(Spawn Probability=0.6, SpawnGroup=1, SpawnRate=32, Use Spawn Probability=true) {module=SpawnRate001}
  ]
  particleSpawnScript [
   /SetVariables(Particles.VisibilityTag=$Emitter.SpawnGroup) {isSetParameters=true}
   &VC_PaintedVariantsOverride(
    Black=<rgba 0.022 0.022 0.022 1>
    BurntSienna=<rgba 0.08 0.0453 0.028 1>
    Cobalt=<rgba 0.055 0.0596 0.11 1>
    Crimson=<rgba 0.08 0.0224 0.024 1>
    DefaultColor=<rgba 0.0339 0.0161 0.23 10>
    ForestGreen=<rgba 0.0176 0.08 0.0207 10>
    Grey=<rgba 0.04 0.04 0.04 1>
    Lime=<rgba 0.0618 0.07 0.021 1>
    Orange=<rgba 0.07 0.0553 0.021 1>
    Pink=<rgba 0.075 0.0465 0.0646 1>
    Purple=<rgba 0.0553 0.021 0.07 1>
    R1 Black:LinearColor=<rgba 0.0191 0.0191 0.0191 1.1>
    R1 BurntSienna:LinearColor=<rgba 0.5 0.3583 0.0749 0.8>
    R1 Cobalt:LinearColor=<rgba 0.18 0.285 0.6 10>
    R1 Crimson:LinearColor=<rgba 0.8 0.32 0.08 2>
    R1 DefaultColor=<rgba 0.56 0.64 0.8 1.5>
    R1 ForestGreen:LinearColor=<rgba 0 1 0.1667 0.8>
    R1 Grey:LinearColor=<rgba 0.1 0.1 0.1 10>
    R1 Lime:LinearColor=<rgba 0.672 0.8 0.16 10>
    R1 Orange:LinearColor=<rgba 0.8 0.6133 0.24 10>
    R1 Pink:LinearColor=<rgba 0.85 0.595 0.7608 10>
    R1 Purple:LinearColor=<rgba 0.7667 0.3 1 10>
    R1 Saffron:LinearColor=<rgba 0.8 0.8 0.08 10>
    R1 SkyBlue:LinearColor=<rgba 0.25 0.5 0.4796 2>
    R1 TitaniumWhite:LinearColor=<rgba 0.52 0.585 0.65 1>
    Saffron=<rgba 0.08 0.0757 0.016 1>
    SkyBlue=<rgba 0.021 0.06 0.0568 1>
    TitaniumWhite=<rgba 0.09 0.09 0.09 10>
    Use R1 Painted Overrides=true
   )
   &InitializeParticle(
    Color Mode=::NewEnumerator0
    Lifetime=&RandomRangeFloat(
     Minimum=&MergeHelper_FloatSwitch(0=0.4, 1=0.5, 2=?, 3=?, Index=$Emitter.SpawnGroup, Num=::NewEnumerator0)
     Maximum=&MergeHelper_FloatSwitch(0=0.65, 1=0.75, 2=?, 3=?, Index=$Emitter.SpawnGroup, Num=::NewEnumerator0)
    )
    Mass Max=2
    Mass Mode=::NewEnumerator3
    Sprite Rotation Mode=::NewEnumerator2
    Sprite Size Mode=::NewEnumerator1
    Uniform Sprite Size Max=90
    Uniform Sprite Size Min=70
    UsePositionOffset=false
    Write Material Random=true
   )
   &PartitionParticles(Number Of Partitions=$Emitter.PortCount)
   &SkeletalMeshLocation(
    Bone / Socket Index=$Particles.PartitionParticles.Partition
    Bone / Socket Sampling Mode=::NewEnumerator5
    MaskSpawnGroup=false
    Mesh Sampling Type=::NewEnumerator6
    Sampled Position Offset=&AddVectorToPosition(
     Position=&MergeHelper_VectorSwitch(0=<-350 0 -70>, 1=<-400 0 -80>, 2=?, 3=?, Index=&SpawnGroup(), Num=::NewEnumerator0)
     Vector=Linked() {linkedVariable {name=User.DTCarOffset, type=NiagaraPosition}}
    )
    Skeletal Mesh=$System.SkeletalMesh
   )
   &ShapeLocation(
    Hemisphere Distribution=<0.5 0.5>
    MaskSpawnGroup=false
    Override Seed=false
    Rotation Mode=::NewEnumerator2
    Sphere Radius=40
    UseOwnerScale=true
    Yaw / Pitch / Roll=<0.25 0.25 0>
   )
   &AddVelocity(
    Cone Angle=80
    Inner Cone Angle=40
    Module.UseConeFalloff=true
    Module.UseVelDistribution=true
    Override Seed=false
    Speed Falloff From Cone Axis=0.3
    Velocity Mode=::NewEnumerator2
    Velocity Speed=&RandomRangeFloat(Minimum=-800, Maximum=-1100)
   )
  ]
  particleUpdateScript [
   &ParticleState()
   &ScaleColor(Scale RGB=&Multiply_VectorByFloat(Vector=<1 1 1>, Float=7), ScaleA=true, ScaleRGB=true) {module=ScaleColor001}
   /SetVariables(
    Particles.DustSize:Vector2f=&Multiply_Vector2D_ByFloat(
     Float=&FloatFromCurve(
      FloatCurve=DataInterface() {propertyValues {Curve {keys [<at 0 0.3 RCIM_Cubic>,<at 1 5 RCIM_Cubic>]}}}
      Scale Curve=&RandomRangeFloat(Minimum=30, Maximum=40)
     )
     Vector2D=<1 1>
    )
    Particles.Position2:NiagaraPosition=&AddVectorToPosition(
     Position=$Particles.Position
     Vector=&MakeVector(X=&RandomRangeFloat(Minimum=-50, Maximum=120), Y=0, Z=0)
    )
    Particles.ScaledR1:LinearColor=&MakeLinearColorFromVectorAndFloat(
     "Vector (RGB)"=&Multiply_VectorByFloat(Vector=&MakeVectorFromLinearColorRGB(Color=$Particles.R1 Color), Float=2)
     "Float (Alpha)"=&FloatFromCurve(
      FloatCurve=DataInterface() {propertyValues {Curve {keys [<at 0 0.5 RCIM_Cubic>, {arriveTangent=-0.5, interpMode=RCIM_Cubic, leaveTangent=-0.5, time=0.5, value=1}, <at 1 0 RCIM_Cubic>]}}}
      Scale Curve=&RandomRangeFloat(Minimum=0.2, Maximum=0.4)
     )
    )
   ) {isSetParameters=true}
   &AccelerationForce(Acceleration=&MergeHelper_VectorSwitch(
     0=&MakeVector(X=-200, Y=&RandomRangeFloat(Minimum=0, Maximum=0), Z=-500)
     1=&MakeVector(X=100, Y=&RandomRangeFloat(Minimum=0, Maximum=0), Z=-400)
     2=?
     3=?
     Index=&SpawnGroup()
     Num=::NewEnumerator0
    ))
   &Drag(Drag=2, Use Linear Drag=true, Use Rotational Drag=false)
   &CurlNoiseForce(Bias Noise=false, Noise Frequency=250, Noise Strength=100, Pan Noise=false)
   &InheritSourceMovement(ApplyForce=false, ApplyPosition=true, ApplyPositionFalloff=false)
   &SolveForcesAndVelocity(Clamp Velocity=false, Limit Acceleration=false)
   &SubUVAnimation(
    "Sprite Renderer "=DataInterface() {propertyValues {SpriteRenderer=&self:ImpactBits_0.NiagaraSpriteRendererProperties_4}}
    UseEndFrame=false
    UseStartFrame=false
   ) {module=SubUVAnimation001}
   &SubUVAnimation(
    "Sprite Renderer "=DataInterface() {propertyValues {SpriteRenderer=&self:ImpactBits_0.NiagaraSpriteRendererProperties_3}}
    UseEndFrame=false
    UseStartFrame=false
   )
  ]
  rendererClasses [&NiagaraSpriteRendererProperties]
  renderers [
   {rendererClass=&NiagaraSpriteRendererProperties}
   {rendererClass=&NiagaraSpriteRendererProperties, rendererIndex=1}
  ]
 }
 Shards {
  emitterUpdateScript [
   &EmitterState()
   &VC_FilteredSocketCount()
   &SpawnRate(SpawnRate=24, Use Spawn Probability=false)
   &SpawnRate(SpawnGroup=1, SpawnRate=24, Use Spawn Probability=false) {module=SpawnRate001}
   &SpawnRate(SpawnGroup=2, SpawnRate=12, Use Spawn Probability=false) {module=SpawnRate002}
  ]
  particleSpawnScript [
   /SetVariables(Particles.VisibilityTag=$Emitter.SpawnGroup) {isSetParameters=true}
   &VC_PaintedVariantsOverride(
    Black=<rgba 0.012 0.012 0.012 5>
    BurntSienna=<rgba 0.15 0.082 0.03 1>
    Cobalt=<rgba 0.08 0.0998 0.2 1>
    Crimson=<rgba 0.2 0.04 0.0453 1>
    DefaultColor=<rgba 0.117 0.1485 0.18 5>
    ForestGreen=<rgba 0.048 0.12 0.0516 1>
    Grey=<rgba 0.1 0.1 0.1 1>
    Lime=<rgba 0.1027 0.12 0.0161 1>
    Orange=<rgba 0.13 0.09 0.0377 1>
    Pink=<rgba 0.14 0.084 0.1196 1>
    Purple=<rgba 0.1006 0.0392 0.14 1>
    R1 Black:LinearColor=<rgba 0.008 0.008 0.008 1>
    R1 BurntSienna:LinearColor=<rgba 0.08 0.0507 0.036 1>
    R1 Cobalt:LinearColor=<rgba 0.0528 0.0584 0.12 1>
    R1 Crimson:LinearColor=<rgba 0.08 0.0336 0.0349 1>
    R1 DefaultColor=<rgba 0.0339 0.0161 0.23 10>
    R1 ForestGreen:LinearColor=<rgba 0.036 0.08 0.0382 1>
    R1 Grey:LinearColor=<rgba 0.048 0.048 0.048 1>
    R1 Lime:LinearColor=<rgba 0.0707 0.08 0.024 1>
    R1 Orange:LinearColor=<rgba 0.08 0.0642 0.0272 1>
    R1 Pink:LinearColor=<rgba 0.1 0.058 0.0847 1>
    R1 Purple:LinearColor=<rgba 0.0684 0.0298 0.085 1>
    R1 Saffron:LinearColor=<rgba 0.1 0.0953 0.03 1>
    R1 SkyBlue:LinearColor=<rgba 0.0405 0.09 0.0859 1>
    R1 TitaniumWhite:LinearColor=<rgba 0.09 0.09 0.09 1>
    Saffron=<rgba 0.12 0.12 0.036 1>
    SkyBlue=<rgba 0.039 0.13 0.123 1>
    TitaniumWhite=<rgba 0.13 0.13 0.13 1>
    Use R1 Painted Overrides=true
   )
   &InitializeParticle(
    Color=$Particles.Color
    Lifetime=&RandomRangeFloat(
     Minimum=&MergeHelper_FloatSwitch(0=0.35, 1=0.5, 2=0.4, 3=?, Index=$Emitter.SpawnGroup, Num=::NewEnumerator1)
     Maximum=&MergeHelper_FloatSwitch(0=0.5, 1=0.7, 2=0.8, 3=?, Index=$Emitter.SpawnGroup, Num=::NewEnumerator1)
    )
    Sprite Rotation Mode=::NewEnumerator2
    Sprite Size=&SelectVector2DFromArray(
     Array Sampling Mode=::NewEnumerator0
     Direct Array Index=?
     Vector 2D Selection Array=DataInterface() {propertyValues {FloatData [<30 80>, <35 180>, <25 120>]}}
     Random Seed=0
    )
    Sprite Size Mode=::NewEnumerator4
    UsePositionOffset=false
    Write Material Random=true
   )
   &PartitionParticles(Number Of Partitions=$Emitter.PortCount)
   &SkeletalMeshLocation(
    Bone / Socket Index=$Particles.PartitionParticles.Partition
    Bone / Socket Sampling Mode=::NewEnumerator5
    MaskSpawnGroup=false
    Mesh Sampling Type=::NewEnumerator6
    Sampled Position Offset=&AddVectorToPosition(
     Position=&MergeHelper_VectorSwitch(0=<-22 0 0>, 1=<-50 0 0>, 2=<-40 0 0>, 3=?, Index=$Emitter.SpawnGroup, Num=::NewEnumerator1)
     Vector=Linked() {linkedVariable {name=User.DTCarOffset, type=NiagaraPosition}}
    )
    Skeletal Mesh=$System.SkeletalMesh
   )
   &ShapeLocation(
    MaskSpawnGroup=false
    Override Seed=false
    Sphere Radius=&MergeHelper_FloatSwitch(0=25, 1=32, 2=30, 3=?, Index=$Emitter.SpawnGroup, Num=::NewEnumerator1)
    UseOwnerScale=true
   )
   &AddVelocity(
    Cone Angle=&MergeHelper_FloatSwitch(0=0, 1=25, 2=50, 3=?, Index=$Emitter.SpawnGroup, Num=::NewEnumerator1)
    Cone Axis=<1 0 -0.2>
    Distribution Along Cone Axis=0.3
    Inner Cone Angle:NiagaraFloat=&MergeHelper_FloatSwitch(0=0, 1=15, 2=0, 3=?, Index=$Emitter.SpawnGroup, Num=::NewEnumerator1)
    Module.UseConeFalloff=true
    Module.UseVelDistribution=true
    Override Seed=false
    Speed Falloff From Cone Axis=0.3
    Velocity Mode=::NewEnumerator2
    Velocity Speed=&MergeHelper_FloatSwitch(0=0, 1=-2200, 2=-1200, 3=?, Index=$Emitter.SpawnGroup, Num=::NewEnumerator1)
   ) {module=AddVelocity001}
   &AddVelocity(
    Velocity=&MergeHelper_VectorSwitch(
     0=&MakeCustomVectorFromBool(Bool=$User.bIsFrontendPreview, True Vector=<-500 0 0>, False Vector=<-50 0 0>)
     1=<0 0 0>
     2=<0 0 0>
     3=?
     Index=$Emitter.SpawnGroup
     Num=::NewEnumerator1
    )
    Velocity Speed Scale=0.75
   )
  ]
  particleUpdateScript [
   &ParticleState()
   &ScaleSpriteSize(Uniform Curve Sprite Scale=DataInterface() {propertyValues {Curve {keys [
        {arriveTangent=1.9204, interpMode=RCIM_Cubic, leaveTangent=1.9204, tangentMode=RCTM_User, time=0, value=0.2}
        {interpMode=RCIM_Cubic, tangentMode=RCTM_User, time=0.3682, value=1}
        {tangentMode=RCTM_User, time=1, value=0}
       ]}}})
   /SetVariables(
    Particles.ScaledR1:LinearColor=&MakeLinearColorFromVectorAndFloat(
     "Vector (RGB)"=&Multiply_VectorByFloat(Vector=&MakeVectorFromLinearColorRGB(Color=$Particles.R1 Color), Float=7)
     "Float (Alpha)"=1
    )
    Particles.ShardsSize:Vector2f=&Multiply_Vector2D_ByFloat(Float=&RandomRangeFloat(Minimum=40, Maximum=60), Vector2D=<1 1>)
    Particles.ShowerSize:Vector2f=&Multiply_Vector2D_ByFloat(
     Float=&FloatFromCurve(FloatCurve=DataInterface() {propertyValues {Curve {keys [
          <at 0 0 RCIM_Cubic>
          {arriveTangentWeight=0.0648, interpMode=RCIM_Cubic, leaveTangentWeight=0.3239, tangentMode=RCTM_User, time=0.1, value=1}
          {arriveTangentWeight=0.2491, interpMode=RCIM_Cubic, leaveTangentWeight=0.1993, tangentMode=RCTM_User, time=0.6, value=1}
          <at 1 0 RCIM_Cubic>
         ]}}})
     Vector2D=&RandomRangeVector2D(Minimum=<15 80>, Maximum=<20 120>)
    )
   ) {isSetParameters=true}
   &ScaleColor(
    Scale Alpha=&FloatFromCurve(
     FloatCurve=DataInterface() {propertyValues {Curve {defaultValue=1, keys [{tangentMode=RCTM_User, time=0, value=0}, {tangentMode=RCTM_User, time=0.5, value=1}]}}}
     Scale Curve=3
    )
    Scale RGB=&Multiply_VectorByFloat(Vector=<1 1 1>, Float=5)
    ScaleA=true
    ScaleRGB=true
   )
   &InheritSourceMovement(Applied Force Scale=<-1 -1 -1>, ApplyForce=false, ApplyPosition=true, ApplyPositionFalloff=false) {module=InheritSourceMovement001}
   &AccelerationForce(Acceleration=&MergeHelper_VectorSwitch(0=<0 0 -120>, 1=<0 0 -850>, 2=<0 0 -850>, 3=?, Index=&SpawnGroup(), Num=::NewEnumerator1))
   &Drag(
    Drag=&MergeHelper_FloatSwitch(0=2, 1=4, 2=1, 3=?, Index=$Emitter.SpawnGroup, Num=::NewEnumerator1)
    Use Linear Drag=true
    Use Rotational Drag=false
   )
   &SolveForcesAndVelocity(Clamp Velocity=false, Limit Acceleration=false)
   &SubUVAnimation(
    "Sprite Renderer "=DataInterface() {propertyValues {SpriteRenderer=&self:Shards_0.NiagaraSpriteRendererProperties_4}}
    UseEndFrame=false
    UseStartFrame=false
   )
   &SubUVAnimation(
    "Sprite Renderer "=DataInterface() {propertyValues {SpriteRenderer=&self:Shards_0.NiagaraSpriteRendererProperties_0}}
    UseEndFrame=false
    UseStartFrame=false
   ) {module=SubUVAnimation001}
   &SubUVAnimation(
    "Sprite Renderer "=DataInterface() {propertyValues {SpriteRenderer=&self:Shards_0.NiagaraSpriteRendererProperties_3}}
    UseEndFrame=false
    UseStartFrame=false
   ) {module=SubUVAnimation002}
  ]
  rendererClasses [&NiagaraSpriteRendererProperties]
  renderers [
   {rendererClass=&NiagaraSpriteRendererProperties}
   {rendererClass=&NiagaraSpriteRendererProperties, rendererIndex=1}
   {rendererClass=&NiagaraSpriteRendererProperties, rendererIndex=2}
  ]
 }
}
summary {
 systemName=NS_Boost_IceShards
 emitters [
  {emitterName=BaseMist, rendererClasses [&NiagaraSpriteRendererProperties]}
  {emitterName=Shards, rendererClasses [&NiagaraSpriteRendererProperties]}
  {emitterName=ImpactBits, rendererClasses [&NiagaraSpriteRendererProperties]}
 ]
 userVariables [
  User.bIsFrontendPreview {type=NiagaraBool}
  User.bIsPainted {type=NiagaraBool}
  User.ColorVariantID {type=NiagaraInt32}
  User.DTCarOffset {type=NiagaraPosition, description="", defaultValue {}}
  User.GameModeID {type=NiagaraInt32}
  User.PaintedColor {type=LinearColor, description="", defaultValue {struct=LinearColor, value=<rgba 1 1 1 1>}}
  User.PaintedScalar {type=NiagaraFloat, defaultValue=1, description=""}
  User.PortCount {type=NiagaraInt32, defaultValue=2}
 ]
}
